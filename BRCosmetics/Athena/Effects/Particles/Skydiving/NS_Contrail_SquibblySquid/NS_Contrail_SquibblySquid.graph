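# Niagara skydiving contrail for the SquibblySquid cosmetic, with water trails at the feet and wobbling ropes from the left hand.
# Foot-anchored foam, splash meshes and ribbons form the water, with center splash sprites following the Water_Center ribbon.
# Rope ribbons are procedurally offset by velocity and sine sway, and Tassel meshes read them to attach and orient.
# Dive alpha scales rates and motion, while front-end preview and glide-lean parameters adjust directions.
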NiagaraSystem /BRCosmetics/Athena/Effects/Particles/Skydiving/NS_Contrail_SquibblySquid/NS_Contrail_SquibblySquid
dependencies {
 usedDataInterfaces [
  &NiagaraDataInterfaceArrayInt32
  &NiagaraDataInterfaceArrayFloat3
  &NiagaraDataInterfaceCurve
  &NiagaraDataInterfaceParticleRead
  &NiagaraDataInterfaceColorCurve
  &NiagaraDataInterfaceArrayFloat
  &NiagaraDataInterfaceVectorCurve
  &NiagaraDataInterfaceCamera
  &NiagaraDataInterfaceSpriteRendererInfo
  &NiagaraDataInterfaceSocketReader
 ]
 usedDynamicInputs [
  &TransformPosition
  /Niagara/DynamicInputs/Math/Clamp_Vector
  &MakeLinearColorFromVectorAndFloat
  &MergeHelper_ColorSwitch
 ]
 usedModules [
  &self:SystemSpawnScript.SystemScriptSource.SystemScriptGraph.NiagaraNodeAssignment_1.SetVariables_#1
  &self:SystemSpawnScript.SystemScriptSource.SystemScriptGraph.NiagaraNodeAssignment_2.SetVariables_#2
  &self:SystemSpawnScript.SystemScriptSource.SystemScriptGraph.NiagaraNodeAssignment_6.SetVariables_#3
  &SystemState
  &self:SystemSpawnScript.SystemScriptSource.SystemScriptGraph.NiagaraNodeAssignment_3.SetVariables_#4
  &self:SystemSpawnScript.SystemScriptSource.SystemScriptGraph.NiagaraNodeAssignment_4.SetVariables_#5
  &self:SystemSpawnScript.SystemScriptSource.SystemScriptGraph.NiagaraNodeAssignment_5.SetVariables_#6
  &self:SystemSpawnScript.SystemScriptSource.SystemScriptGraph.NiagaraNodeAssignment_0.SetVariables_#7
  /Game/Effects/Niagara/Modules/Contrails/FN_Contrail_Setup
  &self:SystemSpawnScript.SystemScriptSource.SystemScriptGraph.NiagaraNodeAssignment_7.SetVariables_#8
  &self:SystemSpawnScript.SystemScriptSource.SystemScriptGraph.NiagaraNodeAssignment_8.SetVariables_#9
 ]
}
emitters {
 FoamParticles {
  emitterUpdateScript [
   &EmitterState()
   &SpawnRate(SpawnGroup=2, SpawnRate=&Lerp_Float(A=200, B=350, Alpha=$System.ContrailDiveAlpha), Use Spawn Probability=false) {module=SpawnRate002}
   &SpawnRate(SpawnGroup=1, SpawnRate=50, Use Spawn Probability=false) {module=SpawnRate001}
   &SpawnRate(SpawnRate=30, Use Spawn Probability=false)
  ]
  particleSpawnScript [
   &PartitionParticles()
   &InitializeParticle(
    Color=<rgba 0.9948 0.998 1 1>
    Lifetime=&MergeHelper_FloatSwitch(
     0=&RandomRangeFloat(Minimum=0.35, Maximum=0.4)
     1=&RandomRangeFloat(Minimum=0.2, Maximum=0.25)
     2=&RandomRangeFloat(Minimum=0.2, Maximum=0.4)
     3=?
     Index=&SpawnGroup()
     Num=::NewEnumerator1
    )
    Sprite Rotation Mode=::NewEnumerator2
    Sprite Size Mode=::NewEnumerator3
    Uniform Sprite Size=&MergeHelper_FloatSwitch(
     0=&RandomRangeFloat(Minimum=15, Maximum=19)
     1=&RandomRangeFloat(Minimum=30, Maximum=40)
     2=&RandomRangeFloat(Minimum=2, Maximum=4)
     3=?
     Index=&SpawnGroup()
     Num=::NewEnumerator0
    )
    UsePositionOffset=false
    Write Material Random=true
   )
   &AddVelocity(Velocity=&MergeHelper_VectorSwitch(
     0=&MakeCustomVectorFromBool(
      Bool=$User.bIsFrontEndPreview
      True Vector=&Lerp_Vector(
       A=&RandomRangeVector(Minimum=<-80 -80 150>, Maximum=<80 80 200>)
       B=&RandomRangeVector(Minimum=<-80 -200 150>, Maximum=<80 -300 200>)
       Alpha=$User.ParaGlideLeanAlpha
      )
      False Vector=&RandomRangeVector(Minimum=<-80 -80 100>, Maximum=<80 80 130>)
     )
     1=&MakeCustomVectorFromBool(
      Bool=$User.bIsFrontEndPreview
      True Vector=&Lerp_Vector(
       A=&RandomRangeVector(Minimum=<-35 -35 75>, Maximum=<35 35 100>)
       B=&RandomRangeVector(Minimum=<-35 -200 150>, Maximum=<35 -300 200>)
       Alpha=$User.ParaGlideLeanAlpha
      )
      False Vector=&RandomRangeVector(Minimum=<-35 -35 75>, Maximum=<35 35 100>)
     )
     2=&MakeCustomVectorFromBool(
      Bool=$User.bIsFrontEndPreview
      True Vector=&Lerp_Vector(
       A=&RandomRangeVector(Minimum=<-80 -80 300>, Maximum=<80 80 500>)
       B=&RandomRangeVector(Minimum=<-80 -400 300>, Maximum=<80 -500 500>)
       Alpha=$User.ParaGlideLeanAlpha
      )
      False Vector=&RandomRangeVector(Minimum=<-80 -80 300>, Maximum=<80 80 500>)
     )
     3=?
     Index=&SpawnGroup()
     Num=::NewEnumerator1
    ))
   &SocketLocation(
    Orientation Sampling=::NewEnumerator1
    SamplePosOffset=true
    Sampled Position Offset=&SelectVectorFromArray(
     Direct Array Index=$Particles.PartitionParticles.Partition
     Vector Selection Array=DataInterface() {propertyValues {FloatData [<0 0 0>, <0 0 0>]}}
    )
    Sampled Position Offset Coordinate Space:ENiagaraSkelMeshTransforms=::NewEnumerator2
    Socket Index=$Particles.PartitionParticles.Partition
    Socket Reader=$System.Socket Reader Feet
    Socket Sampling Mode=::NewEnumerator3
    VelocitySpeedLimit=false
   )
   &SubUVAnimation(
    Evaluation Type=::NewEnumerator0
    "Sprite Renderer "=DataInterface() {propertyValues {SpriteRenderer=&self:FoamParticles_0.NiagaraSpriteRendererProperties_0}}
    SubUV Animation Mode:ENiagara_SubUVLookupModeV2=::NewEnumerator2
    UseEndFrame=false
    UseStartFrame=false
   )
   /Niagara/Modules/Spawn/Location/SphereLocation(MaskSpawnGroup=true, Module.Override Randomness=false, Sphere Radius=17, Surface Only=false)
   &SphereLocation(MaskSpawnGroup=true, Module.Override Randomness=false, Spawn Group Mask=2, Sphere Radius=17, Surface Only=false) {module=SphereLocation001}
   &CameraOffset(Camera Offset Amount=5)
   /SetVariables(
    Particles.Droplet Sprite Color:LinearColor=<rgba 0.5885 1 0.9631 1>
    Particles.Droplet Sprite Size:Vector2f=&Vector2DFromFloat(Value=&RandomRangeFloat(Minimum=1, Maximum=2))
   ) {isSetParameters=true}
   /SetVariables(Particles.VisibilityTag=&SpawnGroup()) {isSetParameters=true}
  ]
  particleUpdateScript [
   &ParticleState()
   &ScaleSpriteSize(Uniform Curve Sprite Scale=DataInterface() {propertyValues {Curve {keys [{arriveTangent=2.7116, interpMode=RCIM_Cubic, leaveTangent=2.7116, tangentMode=RCTM_User, time=0, value=0}, <at 1 1 RCIM_Cubic>]}}})
   &ScaleColor(Scale Mode=2)
   &DynamicMaterialParameters(
    Index 0 Param 1=&RandomRangeFloat(Minimum=0, Maximum=1)
    Index 0 Param 3=&FloatFromCurve(FloatCurve=DataInterface() {propertyValues {Curve {keys [
         {arriveTangent=1.4815, interpMode=RCIM_Cubic, leaveTangent=1.4815, tangentMode=RCTM_User, time=0, value=0}
         {interpMode=RCIM_Cubic, tangentMode=RCTM_User, time=1, value=1}
        ]}}})
    Index 0 Param 4=&FloatFromCurve(FloatCurve=DataInterface() {propertyValues {Curve {keys [
         {arriveTangent=-0.0155, interpMode=RCIM_Cubic, leaveTangent=-0.0155, tangentMode=RCTM_User, time=0.3, value=0}
         {arriveTangent=-0.0133, interpMode=RCIM_Cubic, leaveTangent=-0.0133, tangentMode=RCTM_User, time=1, value=1}
        ]}}})
    Module.Index0Float0Write=true
    Module.Index0Float1Write=true
    Module.Index0Float2Write=true
    Module.Index0Float3Write=true
   )
   &SolveForcesAndVelocity(Clamp Velocity=false, Limit Acceleration=false)
   /SetVariables(
    Particles.Droplet DynamicMaterialParam 1:Vector4f=&MakeVector4(X=1, Y=0, Z=0, W=0)
    Particles.Droplet Velocity:Vector3f=/Niagara/DynamicInputs/ScaleByCurve/ScaleVectorByCurve(
     InitialValue:Vector3f=&UniformRangedVector(Minimum=<-10 -10 0.05>, Maximum=<10 10 2>)
     ScaleCurve:NiagaraDataInterfaceVectorCurve=DataInterface() {propertyValues {
       XCurve {keys [<at 0 10 RCIM_Cubic>, {arriveTangent=-27.3551, interpMode=RCIM_Cubic, leaveTangent=-27.3551, tangentMode=RCTM_User, time=1, value=30}]}
       YCurve {keys [<at 0 10 RCIM_Cubic>, {arriveTangent=-28.1988, interpMode=RCIM_Cubic, leaveTangent=-28.1988, tangentMode=RCTM_User, time=1, value=30}]}
       ZCurve {keys [<at 0 100>,<at 1 400>]}
      }}
     CurveIndex:NiagaraFloat=$Particles.NormalizedAge
    )
   ) {isSetParameters=true}
  ]
  rendererClasses [&NiagaraSpriteRendererProperties]
  renderers [
   {rendererClass=&NiagaraSpriteRendererProperties}
   {rendererClass=&NiagaraSpriteRendererProperties, rendererIndex=1}
   {rendererClass=&NiagaraSpriteRendererProperties, rendererIndex=2}
  ]
 }
 RopeRibbon {
  emitterUpdateScript [
   &EmitterState(
    Inactive Response=::NewEnumerator1
    Life Cycle Mode=::NewEnumerator1
    Loop Behavior=::NewEnumerator1
    Loop Duration Mode=::NewEnumerator1
    UseLoopDelay=false
   )
   &SpawnBurst_Instantaneous(
    Spawn Count=&Multiply_Int(A=$System.ChainNumber, B=$System.ParticlePerRibbon)
    Use Spawn Probability=false
    UseLoopCountLimit=false
   )
  ]
  particleSpawnScript [
   &InitializeParticle(
    Color=<rgba 0.3281 0.1102 0 1>
    Lifetime=5
    Mesh Scale Mode=::NewEnumerator4
    Ribbon Width=3
    Ribbon Width Mode=::NewEnumerator1
    Sprite Size Mode=::NewEnumerator3
    UsePositionOffset=false
    Write Material Random=true
   )
   &PartitionParticles(
    Number Of Partitions=$System.ChainNumber
    Particles Per Partition:NiagaraInt32=$System.ParticlePerRibbon
    Partition Mode:ENiagara_PartitionMode_v2=::NewEnumerator1
   )
   /SetVariables(Particles.NormalizedChainPosition:NiagaraFloat=$Particles.PartitionParticles.Normalized Index In Partition) {isSetParameters=true, module=SetVariables_#10}
   /SetVariables() {isSetParameters=true}
  ]
  particleUpdateScript [
   &ParticleState(Kill Particles When Lifetime Has Elapsed=false)
   &SkeletalMeshLocation(
    Bone / Socket Index=$Particles.PartitionParticles.Partition
    Bone Sampling Mode=::NewEnumerator4
    MaskSpawnGroup=false
    Orientation Sampling=::NewEnumerator1
    Sampled Position Offset=&FN_FrontEndVector3Modifier(Gameplay=<-1 -2 0>, Front End=&Lerp_Vector(A=<-1 -2 0>, B=<8 5 0>, Alpha=$System.FN_Contrail_Setup.Dive Value))
    Sampled Position Offset Coordinate Space:ENiagaraSkelMeshTransforms=::NewEnumerator2
    Skeletal Mesh=DataInterface() {propertyValues {
      DefaultMesh=null
      ExcludeBoneName=root
      OptionalComponentClass=null
      PreviewMesh=&M_Med_Soldier_04_Preview
      SkinningMode=SkinOnTheFly
      SoftSourceActor=null
      SourceMode=Default
      UvSetIndex=0
      WholeMeshLOD=-1
      bExcludeBone=false
      bReadDeformedGeometry=true
      bRequireCurrentFrameData=true
      ComponentTags []
      FilteredBones [hand_l]
      FilteredSockets []
      MeshUserParameter {parameter {name=None, typeDefHandle {registeredTypeIndex=1}, varData []}}
      SamplingRegions []
     }}
   ) {module=SkeletalMeshLocation002}
   /SetVariables(Particles.Position=&AddVectorToPosition(
     Position=$Particles.Position
     Vector=&Multiply_VectorByFloat(
      Vector=&Add_Vector(
       A=&Multiply_Vector(
        A=&NormalizeVector(
         Vector To Normalize=&FN_FrontEndVector3Modifier(Gameplay=$System.PlayerVelDir, Front End=$System.FrontendVelocity)
         Safe Normalize=true
         Safe Fallback Threshold=0
         Safe Fallback Vector=<0 0 0.001>
        )
        B=&SelectVectorFromArray(
         Direct Array Index=$Particles.PartitionParticles.Partition
         Vector Selection Array=DataInterface() {propertyValues {FloatData [<-0.5 -0.5 -1>, <-0.5 -0.5 -1>, <0.1 -1.5 -1>, <0.1 -0.5 -1>, <-0.5 -1.5 -1>, <-0.5 -1 -1>, <0.1 -1 -1>, <0.1 -0.5 -1>]}}
        )
       )
       B=&Multiply_VectorByFloat(
        Vector=&TransformVector(
         Vector=&SelectVectorFromArray(
          Direct Array Index=$Particles.PartitionParticles.Partition
          Direct Array Mode=::NewEnumerator1
          Vector Selection Array=DataInterface() {propertyValues {FloatData [<2 -2 0>, <-2 0 0>, <1 2 0>, <-0.5 0 0>, <0.5 0 0>, <-1 2 0>, <1 1 0>, <-2 1 0>]}}
         )
         Destination Space=::Simulation
        )
        Float=&SelectFloatFromArray(
         Direct Array Index=&SpawnGroup()
         Float Selection Array=DataInterface() {propertyValues {FloatData [0.1,0.15,0.08,0.16,0.15,0.18]}}
        )
       )
      )
      Float=&FloatFromCurve(
       FloatCurve=DataInterface() {propertyValues {Curve {keys [{tangentMode=RCTM_User, time=0, value=0}, {tangentMode=RCTM_User, time=1, value=1}]}}}
       CurveIndex=$Particles.PartitionParticles.Normalized Index In Partition
       Scale Curve=&SelectFloatFromArray(
        Direct Array Index=&SpawnGroup()
        Float Selection Array=DataInterface() {propertyValues {FloatData [97,97,120,100,75,55,94,80]}}
       )
      )
     )
    )) {isSetParameters=true}
   /SetVariables(Particles.Position=&AddVectorToPosition(
     Position=$Particles.Position
     Vector=&Multiply_VectorByFloat(
      Vector=/Niagara/DynamicInputs/Vectors/RotateVector(
       Vector To Rotate:Vector3f=<10 0 3>
       Yaw:NiagaraFloat=&Add_Float(
        A=$Particles.Age
        B=&Multiply_Float(A=&MakeFloatFromInt(Int=&SpawnGroup()), B=2)
       )
       Pitch:NiagaraFloat=0
       Roll:NiagaraFloat=0
       Rotation Rate:NiagaraFloat=&SelectFloatFromArray(
        Direct Array Index=$Particles.PartitionParticles.Partition
        Float Selection Array=DataInterface() {propertyValues {FloatData [65,100,130,120,160,150,120,140]}}
       )
      )
      Float=&Multiply_Float(
       A=&Sine(
        Normalized Angle=&Add_Float(
         A=&Multiply_Float(
          A=$Emitter.Age
          B=&SelectFloatFromArray(
           Direct Array Index=&SpawnGroup()
           Float Selection Array=DataInterface() {propertyValues {FloatData [-2,-2.5,-2.5,-3,-3,-2,-2.5,-2]}}
          )
         )
         B=$Particles.PartitionParticles.Normalized Index In Partition
        )
        Period=0.9
        Scale=&Lerp_Float(A=0.8, B=1.4, Alpha=$System.FN_Contrail_Setup.Dive Value)
        Bias=0
       )
       B=&FloatFromCurve(
        FloatCurve=DataInterface() {propertyValues {Curve {keys [
            {arriveTangent=1.1676, interpMode=RCIM_Cubic, leaveTangent=1.1676, tangentMode=RCTM_User, time=0, value=0}
            {arriveTangent=1.4544, interpMode=RCIM_Cubic, leaveTangent=1.4544, tangentMode=RCTM_User, time=1, value=1}
           ]}}}
        CurveIndex=$Particles.PartitionParticles.Normalized Index In Partition
        Scale Curve=0.6
       )
      )
     )
    )) {isSetParameters=true}
   /SetVariables(Particles.Position=&AddVectorToPosition(
     Position=$Particles.Position
     Vector=&Multiply_VectorByFloat(
      Vector=&SelectVectorFromArray(
       Direct Array Index=$Particles.PartitionParticles.Partition
       Direct Array Mode=::NewEnumerator1
       Vector Selection Array=DataInterface() {propertyValues {FloatData [<-20 40 79>, <40 20 130>, <-25 0 0>, <25 0 0>]}}
      )
      Float=&FloatFromCurve(
       FloatCurve=DataInterface() {propertyValues {Curve {keys [{arriveTangent=0.9713, time=0, value=0}, <at 0.125 0.2>, <at 0.95 0.92>, {interpMode=RCIM_Cubic, leaveTangent=0.018, time=1, value=1}]}}}
       CurveIndex=$Particles.Initial.PartitionParticles.Normalized Index In Partition
       Scale Curve=&MaskFloat(
        Mask Source=1
        Mask Parameter=$Particles.PartitionParticles.Partition
        Mask=0
        Masked Float=&FloatFromCurve(
         FloatCurve=DataInterface() {propertyValues {Curve {
            defaultValue=1
            keys [
             {arriveTangent=0.4561, interpMode=RCIM_Cubic, leaveTangent=0.4561, tangentMode=RCTM_User, time=0, value=0}
             {arriveTangent=1, interpMode=RCIM_Cubic, leaveTangent=1, time=0.3297, value=0.3061}
             {arriveTangent=0.0854, interpMode=RCIM_Cubic, leaveTangent=0.0854, tangentMode=RCTM_User, time=1, value=1}
            ]
           }}}
         CurveIndex=$Engine.System.Age
        )
        Passthrough Float=&FloatFromCurve(
         FloatCurve=DataInterface() {propertyValues {Curve {
            defaultValue=1
            keys [
             {arriveTangent=-0.0467, interpMode=RCIM_Cubic, leaveTangent=-0.0467, tangentMode=RCTM_User, time=0, value=0}
             {arriveTangent=2.5859, interpMode=RCIM_Cubic, leaveTangent=2.5859, tangentMode=RCTM_User, time=0.196, value=0.1441}
             <at 1 1 RCIM_Cubic>
            ]
           }}}
         CurveIndex=$Engine.System.Age
        )
       )
      )
     )
    )) {isSetParameters=true}
  ]
  rendererClasses [&NiagaraRibbonRendererProperties]
  renderers [{rendererClass=&NiagaraRibbonRendererProperties}]
 }
 SplashMeshes {
  emitterUpdateScript [
   &EmitterState()
   &SpawnRate(SpawnRate=&Lerp_Float(A=15, B=20, Alpha=$System.ContrailDiveAlpha), Use Spawn Probability=false)
   &self:ScratchModule(Target Vector:Vector3f=$System.LookAtVector) {module=ScratchModule}
  ]
  particleSpawnScript [
   &PartitionParticles()
   &InitializeParticle(
    Color=<rgba 0.2212 0.6851 1 1>
    Lifetime=0.3
    Mesh Scale Max=<1.3 1.4 1.4>
    Mesh Scale Min=<1.1 1.3 1.3>
    Mesh Scale Mode=::NewEnumerator2
    Position=&Add_Vector(A=/Niagara/DynamicInputs/Helpers/SimulationPosition(), B=<0 0 0>)
    Position Mode=::NewEnumerator1
    UsePositionOffset=false
    Write Material Random=true
   )
   &InitialMeshOrientation(
    Mesh Axis To Orient=<0 0 1>
    Mesh Orientation Mode=::NewEnumerator2
    Orientation Coordinate Space=::Local
    Orientation Vector=$System.LookAtVector
    Rotation=&RandomRangeVector(Minimum=<-1 0 0>, Maximum=<1 0 0>)
    Rotation Coordinate Space=::NewEnumerator2
    Use Rotation Vector=true
   )
   &CameraOffset()
  ]
  particleUpdateScript [
   &ParticleState()
   &SkeletalMeshLocation(
    Bone / Socket Index=$Particles.PartitionParticles.Partition
    Bone Sampling Mode=::NewEnumerator4
    MaskSpawnGroup=false
    Orientation Sampling=::NewEnumerator1
    Sampled Position Offset=&SelectVectorFromArray(
     Direct Array Index=$Particles.PartitionParticles.Partition
     Vector Selection Array=DataInterface() {propertyValues {FloatData [<0 -5 7>, <0 -5 7>]}}
    )
    Sampled Position Offset Coordinate Space:ENiagaraSkelMeshTransforms=::NewEnumerator1
    Skeletal Mesh=DataInterface() {propertyValues {
      DefaultMesh=null
      ExcludeBoneName=root
      OptionalComponentClass=null
      PreviewMesh=&M_Med_Soldier_04_Preview
      SkinningMode=SkinOnTheFly
      SoftSourceActor=null
      SourceMode=Default
      UvSetIndex=0
      WholeMeshLOD=-1
      bExcludeBone=false
      bReadDeformedGeometry=true
      bRequireCurrentFrameData=true
      ComponentTags []
      FilteredBones [foot_l,foot_r]
      FilteredSockets []
      MeshUserParameter {parameter {name=None, typeDefHandle {registeredTypeIndex=1}, varData []}}
      SamplingRegions []
     }}
   )
   &UpdateMeshOrientation(
    Facing Direction=&MakeCustomVectorFromBool(
     Bool=$User.bIsFrontEndPreview
     True Vector=&Lerp_Vector(A=<0 1 0>, B=<0 -1 0>, Alpha=$User.ParaGlideLeanAlpha)
     False Vector=<0 -1 0>
    )
    Facing Interpolation=false
    Orientation Method=::NewEnumerator1
    Override Quat=false
    Side Direction:Vector3f=&MakeCustomVectorFromBool(
     Bool=$User.bIsFrontEndPreview
     True Vector=&Lerp_Vector(A=<1.8889 0.0645 1>, B=<-1.7804 -0.8502 0>, Alpha=$User.ParaGlideLeanAlpha)
     False Vector=<0 0 -1>
    )
    Stabilization Mode:ENiagara_OrientVectorCount=::NewEnumerator3
    TransformFacing=false
    TransformSide=false
    TransformUp=false
    Up Direction:Vector3f=&MakeCustomVectorFromBool(
     Bool=$User.bIsFrontEndPreview
     True Vector=&Lerp_Vector(A=<1 0 0>, B=<1 0 1>, Alpha=$User.ParaGlideLeanAlpha)
     False Vector=<-1 0 0>
    )
   )
   /SetVariables(Particles.InterpolatedTrailVector:Vector3f=$Emitter.InterpolatedVector) {isSetParameters=true}
   &InterpolateOverTime(Rate Of Change=0, Target Vector Value:Vector3f=$Particles.InterpolatedTrailVector, Variable Type=::NewEnumerator2)
   &ScaleMeshSize(Scale Factor=&VectorFromCurve(
     VectorCurve=DataInterface() {propertyValues {
       XCurve {keys [{arriveTangent=2.0113, interpMode=RCIM_Cubic, leaveTangent=2.0113, tangentMode=RCTM_User, time=0, value=0}, <at 1 0.7 RCIM_Cubic>]}
       YCurve {keys [{arriveTangent=2.0113, interpMode=RCIM_Cubic, leaveTangent=2.0113, tangentMode=RCTM_User, time=0, value=0}, <at 1 0.7 RCIM_Cubic>]}
       ZCurve {keys [{arriveTangent=2.0113, interpMode=RCIM_Cubic, leaveTangent=2.0113, tangentMode=RCTM_User, time=0, value=0}, <at 1 0.7 RCIM_Cubic>]}
      }}
     Scale Curve=&VectorFromFloat(Value=0.16)
    ))
   &SolveForcesAndVelocity(Clamp Velocity=false, Limit Acceleration=false)
   &ScaleColor(
    Scale Alpha=&FloatFromCurve(FloatCurve=DataInterface() {propertyValues {Curve {keys [
         {arriveTangent=5.6426, interpMode=RCIM_Cubic, leaveTangent=5.6426, tangentMode=RCTM_User, time=0, value=0}
         {interpMode=RCIM_Cubic, tangentMode=RCTM_User, time=0.1, value=1}
         {tangentMode=RCTM_User, time=1, value=0}
        ]}}})
    Scale RGB=<2 2 2>
    ScaleA=true
    ScaleRGB=true
   )
   &DynamicMaterialParameters(
    Index 0 Param 4=409.6645
    Index 1 Param 1=&FloatFromCurve(FloatCurve=DataInterface() {propertyValues {Curve {keys [<at 0 0 RCIM_Cubic>,<at 1 0.9 RCIM_Cubic>]}}})
    Module.Index0Float0Write=true
    Module.Index0Float1Write=true
    Module.Index0Float2Write=true
    Module.Index0Float3Write=true
    Module.Index1Float0Write=true
    Module.Index1Float1Write=true
    Module.Index1Float2Write=true
    Module.Index1Float3Write=true
    Write Parameter Index 1=::NewEnumerator1
   )
  ]
  rendererClasses [&NiagaraMeshRendererProperties]
  renderers [{rendererClass=&NiagaraMeshRendererProperties, rendererIndex=0}]
 }
 Splash_Center {
  emitterUpdateScript [
   &EmitterState()
   &SpawnRate(SpawnRate=&Lerp_Float(A=55, B=75, Alpha=$System.ContrailDiveAlpha), Use Spawn Probability=false)
  ]
  particleSpawnScript [
   &InitializeParticle(
    Lifetime=&Multiply_Float(
     A=&RandomRangeFloat(Minimum=0.35, Maximum=0.45)
     B=&Lerp_Float(A=0.5, B=0.3, Alpha=$System.ContrailDiveAlpha)
    )
    Mesh Scale Mode=::NewEnumerator3
    Sprite Rotation Mode=::NewEnumerator2
    Sprite Size Mode=::NewEnumerator1
    Uniform Sprite Size Max=30
    Uniform Sprite Size Min=20
    UsePositionOffset=false
    Write Material Random=true
   ) {module=InitializeParticle001}
   &PartitionParticles()
   /Niagara/Modules/Update/SubUV/SubUVAnimation(SubUV Animation Mode=::NewEnumerator1)
   &ShapeLocation(
    Cylinder Distribution=::NewEnumerator2
    Cylinder Height=1
    Cylinder Height Midpoint=1
    Cylinder Radius=&RandomRangeFloat(Minimum=20, Maximum=50)
    MaskSpawnGroup=false
    Rotation Axis=<0 0 1>
    Rotation Mode=::NewEnumerator1
    Shape Primitive=::NewEnumerator1
    U Position=&RandomRangeFloat(Minimum=0, Maximum=1)
    UseOwnerScale=true
    V Position=&RandomRangeFloat(Minimum=0, Maximum=1)
   )
   /SetVariables(Particles.VisibilityTag=&SelectIntFromArray(
     Array Sampling Mode=::NewEnumerator1
     Direct Array Index=?
     Int Selection Array=DataInterface() {propertyValues {IntData [0,1,0]}}
     Random Seed=0
    )) {isSetParameters=true}
   &SocketLocation(
    SamplePosOffset=true
    Sampled Position Offset=&SelectVectorFromArray(
     Direct Array Index=$Particles.PartitionParticles.Partition
     Vector Selection Array=DataInterface() {propertyValues {FloatData [<0 0 0>, <0 0 0>]}}
    )
    Sampled Position Offset Coordinate Space:ENiagaraSkelMeshTransforms=::NewEnumerator2
    Socket Index=$Particles.PartitionParticles.Partition
    Socket Reader=$System.Socket Reader Feet
    Socket Sampling Mode=::NewEnumerator3
    VelocitySpeedLimit=false
   )
  ]
  particleUpdateScript [
   &ParticleState()
   &self:FollowRibbon_01(
    Alpha:NiagaraFloat=&Add_Float(
     A=&RandomRangeFloat(Minimum=0, Maximum=0.15)
     B=&FloatFromCurve(
      FloatCurve=DataInterface() {propertyValues {Curve {keys [
          {arriveTangent=0.4215, interpMode=RCIM_Cubic, leaveTangent=0.4215, tangentMode=RCTM_User, time=0, value=0}
          {arriveTangent=1.5182, interpMode=RCIM_Cubic, leaveTangent=1.5182, tangentMode=RCTM_User, time=1, value=1}
         ]}}}
      Scale Curve=&RandomRangeFloat(Minimum=0.5, Maximum=0.6)
     )
    )
    ParticlePerRibbon=5
    RibbonNumber:NiagaraInt32=$Particles.PartitionParticles.Partition
    RibbonReader=DataInterface() {propertyValues {EmitterBinding {bindingMode=Other, emitterName=Water_Center}}}
   ) {module=FollowRibbon_01}
   /SetVariables(Particles.Position=$Output.FollowRibbon_01.RibbonPos) {isSetParameters=true}
   &ScaleSpriteSize(Scale Factor=&Vector2DFromFloat(Value=&FloatFromCurve(FloatCurve=DataInterface() {propertyValues {Curve {keys [{arriveTangent=1.3936, interpMode=RCIM_Cubic, leaveTangent=1.3936, tangentMode=RCTM_User, time=0, value=0.1}, <at 1 1 RCIM_Cubic>]}}})))
   &SpriteRotationRate(Rotation Rate=&FloatFromCurve(
     FloatCurve=DataInterface() {propertyValues {Curve {keys [
         {interpMode=RCIM_Cubic, tangentMode=RCTM_User, time=0, value=0}
         {arriveTangent=1.8494, interpMode=RCIM_Cubic, leaveTangent=1.8494, tangentMode=RCTM_User, time=1, value=1}
        ]}}}
     Scale Curve=&RandomRangeFloat(Minimum=-200, Maximum=200)
    ))
   &DynamicMaterialParameters(
    Index 0 Param 1=&RandomRangeFloat(Minimum=0, Maximum=1)
    Index 0 Param 3=&FloatFromCurve(FloatCurve=DataInterface() {propertyValues {Curve {keys [
         {arriveTangent=1.4815, interpMode=RCIM_Cubic, leaveTangent=1.4815, tangentMode=RCTM_User, time=0, value=0}
         {interpMode=RCIM_Cubic, tangentMode=RCTM_User, time=1, value=1}
        ]}}})
    Index 0 Param 4=&FloatFromCurve(FloatCurve=DataInterface() {propertyValues {Curve {keys [
         {arriveTangent=-0.0155, interpMode=RCIM_Cubic, leaveTangent=-0.0155, tangentMode=RCTM_User, time=0.3, value=0.05}
         {arriveTangent=-0.0133, interpMode=RCIM_Cubic, leaveTangent=-0.0133, tangentMode=RCTM_User, time=0.7, value=0.8}
        ]}}})
    Module.Index0Float0Write=true
    Module.Index0Float1Write=true
    Module.Index0Float2Write=true
    Module.Index0Float3Write=true
   )
   &ScaleColor(
    Scale Alpha=&FloatFromCurve(
     FloatCurve=DataInterface() {propertyValues {Curve {keys [
         {interpMode=RCIM_Cubic, tangentMode=RCTM_User, time=0.2, value=2}
         {arriveTangent=-1.2825, interpMode=RCIM_Cubic, leaveTangent=-1.2825, tangentMode=RCTM_User, time=1, value=0}
        ]}}}
     Scale Curve=&MergeHelper_FloatSwitch(0=1, 1=1, 2=?, 3=?, Index=$Particles.PartitionParticles.Partition, Num=::NewEnumerator0)
    )
    Scale RGB=&VectorFromFloat(Value=&FloatFromCurve(
      FloatCurve=DataInterface() {propertyValues {Curve {keys [
          {interpMode=RCIM_Cubic, tangentMode=RCTM_User, time=0, value=2}
          {arriveTangent=-0.9872, interpMode=RCIM_Cubic, leaveTangent=-0.9872, tangentMode=RCTM_User, time=1, value=0.25}
         ]}}}
      Scale Curve=&FN_FrontEndFloatModifier(Gameplay=1.2, Front End=2)
     ))
    ScaleA=true
    ScaleRGB=true
   )
   /SetVariables(Particles.SpriteSize02:Vector2f=&Multiply_Vector2D_ByFloat(Float=1.5, Vector2D=$Particles.SpriteSize)) {isSetParameters=true}
  ]
  rendererClasses [&NiagaraSpriteRendererProperties]
  renderers [
   {rendererClass=&NiagaraSpriteRendererProperties}
   {rendererClass=&NiagaraSpriteRendererProperties, rendererIndex=1}
  ]
 }
 Tassel {
  emitterUpdateScript [
   &EmitterState(
    Inactive Response=::NewEnumerator1
    Life Cycle Mode=::NewEnumerator1
    Loop Behavior=::NewEnumerator1
    Loop Duration Mode=::NewEnumerator1
    UseLoopDelay=false
   )
   &SpawnBurst_Instantaneous(Spawn Count=&Multiply_Int(A=$System.ChainNumber, B=1), Use Spawn Probability=false, UseLoopCountLimit=false)
  ]
  particleSpawnScript [
   &FN_SetSpawnGroup(Number of Spawn Groups=2)
   &InitializeParticle(
    Lifetime=5
    Mesh Scale Mode=::NewEnumerator4
    Sprite Size Mode=::NewEnumerator3
    Uniform Sprite Size=100
    UsePositionOffset=false
    Write Material Random=true
   )
   &PartitionParticles(Number Of Partitions=$System.ChainNumber, Particles Per Partition=1, Partition Mode:ENiagara_PartitionMode_v2=::NewEnumerator1)
   /SetVariables(Particles.NormalizedChainPosition:NiagaraFloat=$Particles.PartitionParticles.Normalized Index In Partition) {isSetParameters=true, module=SetVariables_#10}
   &SkeletalMeshLocation(
    Bone / Socket Index=$Particles.PartitionParticles.Partition
    Bone Sampling Mode=::NewEnumerator4
    MaskSpawnGroup=false
    Sampled Position Offset=&Lerp_Vector(A=<0 0 0>, B=<0 0 100>, Alpha=$Particles.NormalizedChainPosition)
    Sampled Position Offset Coordinate Space:ENiagaraSkelMeshTransforms=::NewEnumerator1
    Skeletal Mesh=DataInterface() {propertyValues {
      DefaultMesh=null
      ExcludeBoneName=root
      OptionalComponentClass=null
      PreviewMesh=&M_Med_Soldier_04_Preview
      SkinningMode=SkinOnTheFly
      SoftSourceActor=null
      SourceMode=Default
      UvSetIndex=0
      WholeMeshLOD=-1
      bExcludeBone=false
      bReadDeformedGeometry=true
      bRequireCurrentFrameData=true
      ComponentTags []
      FilteredBones [hand_l]
      FilteredSockets []
      MeshUserParameter {parameter {name=None, typeDefHandle {registeredTypeIndex=1}, varData []}}
      SamplingRegions []
     }}
   )
  ]
  particleUpdateScript [
   &ParticleState(Kill Particles When Lifetime Has Elapsed=false)
   &ScaleMeshSize(Scale Factor=&VectorFromFloat(Value=&FloatFromCurve(
      FloatCurve=DataInterface() {propertyValues {Curve {defaultValue=1, keys [{tangentMode=RCTM_User, time=0, value=0}, {tangentMode=RCTM_User, time=0.5, value=1}]}}}
      CurveIndex=$Engine.System.Age
     )))
   &SkeletalMeshLocation(
    Bone / Socket Index=$Particles.PartitionParticles.Partition
    Bone Sampling Mode=::NewEnumerator4
    MaskSpawnGroup=false
    Position Sampling=::NewEnumerator2
    Sampled Position Offset=&Lerp_Vector(A=<20 -3 0>, B=<-20 3 1>, Alpha=&MakeFloatFromInt(Int=$Particles.PartitionParticles.Partition))
    Skeletal Mesh=$System.Skeletal Mesh
   ) {module=SkeletalMeshLocation001}
   &self:StickMeshToRibbon_03(
    Index:NiagaraInt32=$Particles.PartitionParticles.Index In Partition
    LinkOffset=20
    Particle Attribute Reader=DataInterface() {propertyValues {EmitterBinding {bindingMode=Other, emitterName=RopeRibbon}}}
    ParticlePerRibbon=1
    Partition:NiagaraInt32=$Particles.PartitionParticles.Partition
    RibbonSpace=21
   ) {module=StickMeshToRibbon_03}
   &InterpolateOverTime(
    Rate Of Change:NiagaraFloat=&FN_FrontEndFloatModifier(Gameplay=1, Front End=10)
    Target Vector Value:Vector3f=&Multiply_VectorByFloat(
     Vector=&NormalizeVector(
      Vector To Normalize=&Multiply_Vector(
       A=&ReturnCameraVector(Camera Vectors=::NewEnumerator4, Coordinate Space=::Simulation, Origin Position=$Particles.Position)
       B=<1 1 1>
      )
      Safe Normalize=true
      Safe Fallback Threshold=0
      Safe Fallback Vector=<0 0 0.001>
     )
     Float=-1
    )
    Variable Type=::NewEnumerator2
   )
   &InterpolateOverTime(
    Rate Of Change=10
    Target Vector Value:Vector3f=&Multiply_VectorByFloat(Vector=$Particles.RibbonDirectionVector, Float=-1)
    Variable Type=::NewEnumerator2
   ) {module=InterpolateOverTime001}
   &self:OrientChainLinks_03(
    FrontVector:Vector3f=$Particles.InterpolateOverTime.InterpolateOverTime_V3_.Moving Average
    RightVector:Vector3f=<1 0 0>
    UpVector:Vector3f=$Particles.InterpolateOverTime001.InterpolateOverTime_V3_.Moving Average
   ) {module=OrientChainLinks_03}
   &SolveForcesAndVelocity(Clamp Velocity=false, Limit Acceleration=false)
   /SetVariables(Particles.Balloon_Orientation:Quat4f=&MultiplyQuaternion(
     Quaternion A=&MakeQuatFromAxisAngle(
      Axis=<0 0 1>
      AngleInDegrees=&SelectFloatFromArray(Direct Array Index=$Particles.PartitionParticles.Partition, Float Selection Array=DataInterface() {propertyValues {FloatData [5,-10]}})
     )
     Quaternion B=&MakeQuatFromAxisAngle(
      Axis=<0 0 -1>
      AngleInDegrees=&Lerp_Float(
       A=0
       B=&SelectFloatFromArray(Direct Array Index=$Particles.PartitionParticles.Partition, Float Selection Array=DataInterface() {propertyValues {FloatData [-50,50]}})
       Alpha=$User.ParaGlideLeanAlpha
      )
      NormalizeAxis=true
     )
    )) {isSetParameters=true}
   /SetVariables(Particles.Balloon_Orientation:Quat4f=&MultiplyQuaternion(
     Quaternion A=$Particles.Balloon_Orientation
     Quaternion B=&MakeQuatFromAxisAngle(
      Axis=&SelectVectorFromArray(
       Direct Array Index=$Particles.PartitionParticles.Partition
       Vector Selection Array=DataInterface() {propertyValues {FloatData [<-1 1 0>, <1 1 0>]}}
      )
      AngleInDegrees=&Lerp_Float(
       A=0
       B=&MakeCustomFloatFromBool(
        Bool=$User.bIsFrontEndPreview
        True Float=&SelectFloatFromArray(Direct Array Index=$Particles.PartitionParticles.Partition, Float Selection Array=DataInterface() {propertyValues {FloatData [-50,50]}})
        False Float=0
       )
       Alpha=$User.ParaGlideLeanAlpha
      )
      NormalizeAxis=true
     )
    )) {isSetParameters=true}
   /SetVariables(
    Particles.MaterialRotate1Axis:Vector3f=&TransformVector(
     Vector=&RotateVectorByQuaternion(
      Vector To Rotate=&Lerp_Vector(
       A=&SelectVectorFromArray(
        Direct Array Index=$Particles.PartitionParticles.Partition
        Vector Selection Array=DataInterface() {propertyValues {FloatData [<0.1 0.1 0.1>, <0.1 0.1 0>]}}
       )
       B=&SelectVectorFromArray(
        Direct Array Index=$Particles.PartitionParticles.Partition
        Vector Selection Array=DataInterface() {propertyValues {FloatData [<1 1 0>, <1 1 0>]}}
       )
       Alpha=$System.Timeline.NormalizedTime
      )
      Quaternion=$Particles.Balloon_Orientation
     )
     Destination Space=::World
    )
    Particles.MaterialRotateAxis:Vector3f=&TransformVector(
     Vector=&RotateVectorByQuaternion(
      Vector To Rotate=&Lerp_Vector(
       A=&SelectVectorFromArray(
        Direct Array Index=$Particles.PartitionParticles.Partition
        Vector Selection Array=DataInterface() {propertyValues {FloatData [<0.1 0.1 0.1>, <0.85 0.35 0.5>]}}
       )
       B=&SelectVectorFromArray(
        Direct Array Index=$Particles.PartitionParticles.Partition
        Vector Selection Array=DataInterface() {propertyValues {FloatData [<0.1 0.1 0>, <1 0.5 0.75>]}}
       )
       Alpha=$System.Timeline.NormalizedTime
      )
      Quaternion=$Particles.Balloon_Orientation
     )
     Destination Space=::World
    )
   ) {isSetParameters=true}
   &FN_Timer(Increment Multiplier:NiagaraFloat=&Lerp_Float(A=1, B=2, Alpha=$User.ParaGlideLeanAlpha))
   &DynamicMaterialParameters(
    Index 0 Param 1=&MakeFloatFromVector(Vector=$Particles.MaterialRotateAxis, Channel=::NewEnumerator0)
    Index 0 Param 2=&MakeFloatFromVector(Vector=$Particles.MaterialRotateAxis, Channel=::NewEnumerator1)
    Index 0 Param 3=&MakeFloatFromVector(Vector=$Particles.MaterialRotateAxis, Channel=::NewEnumerator2)
    Index 0 Param 4=&RandomRangeFloat(Minimum=0.3, Maximum=0.4)
    Index 1 Param 1=&MakeFloatFromVector(Vector=$Particles.MaterialRotate1Axis, Channel=::NewEnumerator0)
    Index 1 Param 2=&MakeFloatFromVector(Vector=$Particles.MaterialRotate1Axis, Channel=::NewEnumerator1)
    Index 1 Param 3=&MakeFloatFromVector(Vector=$Particles.MaterialRotate1Axis, Channel=::NewEnumerator2)
    Index 1 Param 4=&Lerp_Float(A=5, B=5, Alpha=$System.FN_Contrail_Setup.Dive Value)
    Index 2 Param 1=$Particles.FN_Timer.Value
    Index 2 Param 2=&Lerp_Float(A=0.35, B=0.8, Alpha=$System.FN_Contrail_Setup.Dive Value)
    Index 2 Param 3=&Lerp_Float(A=0.5, B=0.5, Alpha=$System.FN_Contrail_Setup.Dive Value)
    Index 2 Param 4=&Lerp_Float(A=2, B=2.8, Alpha=$System.FN_Contrail_Setup.Dive Value)
    Index 3 Param 1=1
    Index 3 Param 2=1
    Index 3 Param 3=1
    Module.Index0Float0Write=true
    Module.Index0Float1Write=true
    Module.Index0Float2Write=true
    Module.Index0Float3Write=true
    Module.Index1Float0Write=true
    Module.Index1Float1Write=true
    Module.Index1Float2Write=true
    Module.Index1Float3Write=true
    Module.Index2Float0Write=true
    Module.Index2Float1Write=true
    Module.Index2Float2Write=true
    Module.Index2Float3Write=true
    Module.Index3Float0Write=true
    Module.Index3Float1Write=true
    Module.Index3Float2Write=true
    Module.Index3Float3Write=true
    Write Parameter Index 1=::NewEnumerator1
    Write Parameter Index 2=::NewEnumerator1
    Write Parameter Index 3=::NewEnumerator1
   )
  ]
  rendererClasses [&NiagaraMeshRendererProperties]
  renderers [{rendererClass=&NiagaraMeshRendererProperties, rendererIndex=0}]
 }
 WaterTrail_Low {
  emitterSpawnScript [
   /SetVariables(Emitter.InterpolatedVector:Vector3f=<0 0 1>, Emitter.TargetVector:Vector3f=<0 0 1>) {isSetParameters=true, module=SetVariables_#11}
   /SetVariables(Emitter.RibbonLength=222) {isSetParameters=true, module=SetVariables_#12}
   /SetVariables(Emitter.ParticlesPerTrail=5) {isSetParameters=true, module=SetVariables_#13}
  ]
  emitterUpdateScript [
   &EmitterState(
    Inactive Response=::NewEnumerator1
    Life Cycle Mode=::NewEnumerator1
    Loop Behavior=::NewEnumerator1
    Loop Duration Mode=::NewEnumerator1
    UseLoopDelay=false
   )
   &SpawnBurst_Instantaneous(
    Spawn Count=&Multiply_Int(A=4, B=$Emitter.ParticlesPerTrail)
    Use Spawn Probability=false
    UseLoopCountLimit=false
   )
   &self:ScratchModule(Target Vector:Vector3f=$System.LookAtVector) {module=ScratchModule}
  ]
  particleSpawnScript [
   &InitializeParticle(
    Lifetime=2
    Position Mode=::NewEnumerator0
    Ribbon Width=15
    Ribbon Width Mode=::NewEnumerator1
    Sprite Size Mode=::NewEnumerator3
    Uniform Sprite Size=5
    Write Material Random=true
   )
   &PartitionParticles(
    Normalized Index In Partition Includes One=true
    Particles Per Partition:NiagaraInt32=$Emitter.ParticlesPerTrail
    Partition Mode:ENiagara_PartitionMode_v2=::NewEnumerator1
    Write Partition Index to Ribbon Link Order:ENiagara_AttributeSamplingApplyOutput=::NewEnumerator0
    Write Partition to Ribbon ID=::NewEnumerator0
   )
   /SetVariables(Particles.RibbonWidth=&SelectFloatFromArray(Direct Array Index=$Particles.PartitionParticles.Partition, Float Selection Array=DataInterface() {propertyValues {FloatData [15,15]}})) {isSetParameters=true, module=SetVariables_#14}
   /SetVariables(Particles.RibbonLength:NiagaraFloat=&SelectFloatFromArray(Direct Array Index=$Particles.PartitionParticles.Partition, Float Selection Array=DataInterface() {propertyValues {FloatData [140,140]}})) {isSetParameters=true, module=SetVariables_#15}
  ]
  particleUpdateScript [
   &ParticleState(Kill Particles When Lifetime Has Elapsed=false)
   &SocketLocation(
    Orientation Sampling=::NewEnumerator1
    Position Sampling=::NewEnumerator1
    SamplePosOffset=true
    Sampled Position Offset=&SelectVectorFromArray(
     Direct Array Index=$Particles.PartitionParticles.Partition
     Vector Selection Array=DataInterface() {propertyValues {FloatData [<0 0 -5>, <0 0 -5>]}}
    )
    Sampled Position Offset Coordinate Space:ENiagaraSkelMeshTransforms=::NewEnumerator2
    Socket Index=$Particles.PartitionParticles.Partition
    Socket Reader=$System.Socket Reader Feet
    Socket Sampling Mode=::NewEnumerator3
    VelocitySpeedLimit=false
   )
   /SetVariables(Particles.RibbonLength:NiagaraFloat=&Multiply_Float(
     A=$Particles.Initial.RibbonLength
     B=&FloatFromCurve(
      FloatCurve=DataInterface() {propertyValues {Curve {keys [
          {arriveTangent=1, interpMode=RCIM_Cubic, time=0, value=0}
          {interpMode=RCIM_Cubic, tangentMode=RCTM_User, time=0, value=0}
          {interpMode=RCIM_Cubic, tangentMode=RCTM_User, time=1, value=1}
         ]}}}
      Scale Curve=&Lerp_Float(A=1, B=1.2, Alpha=$System.ContrailDiveAlpha)
     )
    )) {isSetParameters=true, module=SetVariables_#16}
   /SetVariables(Particles.InterpolatedTrailVector:Vector3f=$Emitter.InterpolatedVector) {isSetParameters=true, module=SetVariables_#17}
   /SetVariables(Particles.InterpolatedPosition:NiagaraPosition=&AddVectorToPosition(
     Position=$Output.SocketLocation.SampledPosition
     Vector=&Multiply_VectorByFloat(
      Vector=$Particles.InterpolatedTrailVector
      Float=&Multiply_Float(A=$Particles.RibbonLength, B=$Particles.PartitionParticles.Normalized Index In Partition)
     )
    )) {isSetParameters=true, module=SetVariables_#18}
   &ScaleRibbonWidth(Ribbon Width Scale=&Multiply_Float(
     A=&FloatFromCurve(
      FloatCurve=DataInterface() {propertyValues {Curve {keys [
          {interpMode=RCIM_Cubic, tangentMode=RCTM_User, time=0, value=0}
          {arriveTangent=1.4972, interpMode=RCIM_Cubic, leaveTangent=1.4972, tangentMode=RCTM_User, time=1, value=1.3}
         ]}}}
      Scale Curve=1.3
     )
     B=&MergeHelper_FloatSwitch(
      0=&FloatFromCurve(
       FloatCurve=DataInterface() {propertyValues {Curve {keys [
           {interpMode=RCIM_Cubic, tangentMode=RCTM_User, time=0, value=0.2}
           {arriveTangentWeight=0.5252, interpMode=RCIM_Cubic, tangentMode=RCTM_User, time=1, value=1}
          ]}}}
       CurveIndex=$Particles.RibbonLinkOrder
       Scale Curve=1.5
      )
      1=&FloatFromCurve(
       FloatCurve=DataInterface() {propertyValues {Curve {keys [
           {interpMode=RCIM_Cubic, tangentMode=RCTM_User, time=0, value=0.2}
           {arriveTangentWeight=0.5252, interpMode=RCIM_Cubic, tangentMode=RCTM_User, time=1, value=1}
          ]}}}
       CurveIndex=$Particles.RibbonLinkOrder
       Scale Curve=1.5
      )
      2=&FloatFromCurve(
       FloatCurve=DataInterface() {propertyValues {Curve {keys [
           {interpMode=RCIM_Cubic, tangentMode=RCTM_User, time=0, value=0.5}
           {arriveTangentWeight=0.5252, interpMode=RCIM_Cubic, tangentMode=RCTM_User, time=1, value=1}
          ]}}}
       CurveIndex=$Particles.RibbonLinkOrder
      )
      3=&FloatFromCurve(
       FloatCurve=DataInterface() {propertyValues {Curve {keys [
           {interpMode=RCIM_Cubic, tangentMode=RCTM_User, time=0, value=0.5}
           {arriveTangentWeight=0.5252, interpMode=RCIM_Cubic, tangentMode=RCTM_User, time=1, value=1}
          ]}}}
       CurveIndex=$Particles.RibbonLinkOrder
      )
      Index=$Particles.PartitionParticles.Partition
      Num=::NewEnumerator0
     )
    ))
   &Color(Color=&ColorFromCurve(
     ColorCurve=DataInterface() {propertyValues {
       AlphaCurve {keys [<at 0 0>,<at 0.2858 1.5 RCIM_Cubic>,<at 0.9164 0 RCIM_Cubic>]}
       BlueCurve {keys [<at 0.0111 0.8177>,<at 0.3546 0.5>,<at 0.663 0.1 RCIM_Cubic>]}
       GreenCurve {keys [<at 0.0111 0.6534>,<at 0.3546 0.2489>,<at 0.663 0.0351 RCIM_Cubic>]}
       RedCurve {keys [<at 0.0111 0.1704>,<at 0.3546 0.0198>,<at 0.663 0.005 RCIM_Cubic>]}
      }}
     CurveIndex=$Particles.RibbonLinkOrder
    ))
   /SetVariables(Particles.Offset X:NiagaraFloat=&Add_Float(A=$Particles.Offset X, B=&Lerp_Float(A=0, B=-0.03, Alpha=$System.ContrailDiveAlpha))) {isSetParameters=true, module=SetVariables_#19}
   &DynamicMaterialParameters(
    Index 0 Param 1=&SelectFloatFromArray(Direct Array Index=$Particles.PartitionParticles.Partition, Float Selection Array=DataInterface() {propertyValues {FloatData [0.7,0.8]}})
    Index 0 Param 2=$Particles.Offset X
    Index 0 Param 3=&Multiply_Float(
     A=&FloatFromCurve(
      FloatCurve=DataInterface() {propertyValues {Curve {keys [
          {arriveTangent=0.0484, interpMode=RCIM_Cubic, leaveTangent=0.0484, tangentMode=RCTM_User, time=0.3, value=1}
          {arriveTangent=0.0924, interpMode=RCIM_Cubic, leaveTangent=0.0924, tangentMode=RCTM_User, time=0.8, value=1}
          <at 1 0 RCIM_Cubic>
         ]}}}
      CurveIndex=$Particles.PartitionParticles.Normalized Index In Partition
      Scale Curve=0.6835
     )
     B=&SelectFloatFromArray(Direct Array Index=$Particles.PartitionParticles.Partition, Float Selection Array=DataInterface() {propertyValues {FloatData [0.5,0.5]}})
    )
    Index 0 Param 4=1
    Module.Index0Float0Write=true
    Module.Index0Float1Write=true
    Module.Index0Float2Write=true
    Module.Index0Float3Write=true
   )
   /SetVariables(Particles.Position=$Particles.InterpolatedPosition) {isSetParameters=true, module=SetVariables_#20}
  ]
  rendererClasses [&NiagaraRibbonRendererProperties]
  renderers [
   {rendererClass=&NiagaraRibbonRendererProperties}
   {rendererClass=&NiagaraRibbonRendererProperties, rendererIndex=1}
  ]
 }
 Water_Center {
  emitterSpawnScript [
   /SetVariables(Emitter.InterpolatedVector:Vector3f=<0 0 1>, Emitter.TargetVector:Vector3f=<0 0 1>) {isSetParameters=true, module=SetVariables_#11}
   /SetVariables(Emitter.RibbonLength=222) {isSetParameters=true, module=SetVariables_#12}
   /SetVariables(Emitter.ParticlesPerTrail=5) {isSetParameters=true, module=SetVariables_#13}
  ]
  emitterUpdateScript [
   &EmitterState(
    Inactive Response=::NewEnumerator1
    Life Cycle Mode=::NewEnumerator1
    Loop Behavior=::NewEnumerator1
    Loop Duration Mode=::NewEnumerator1
    UseLoopDelay=false
   )
   &SpawnBurst_Instantaneous(
    Spawn Count=&Multiply_Int(A=4, B=$Emitter.ParticlesPerTrail)
    Use Spawn Probability=false
    UseLoopCountLimit=false
   )
   &self:ScratchModule(Target Vector:Vector3f=$System.LookAtVector) {module=ScratchModule}
  ]
  particleSpawnScript [
   &InitializeParticle(
    Lifetime=2
    Position Mode=::NewEnumerator0
    Ribbon Width=15
    Ribbon Width Mode=::NewEnumerator1
    Sprite Size Mode=::NewEnumerator3
    Uniform Sprite Size=5
    Write Material Random=true
   )
   &PartitionParticles(
    Normalized Index In Partition Includes One=true
    Particles Per Partition:NiagaraInt32=$Emitter.ParticlesPerTrail
    Partition Mode:ENiagara_PartitionMode_v2=::NewEnumerator1
    Write Partition Index to Ribbon Link Order:ENiagara_AttributeSamplingApplyOutput=::NewEnumerator0
    Write Partition to Ribbon ID=::NewEnumerator0
   )
   /SetVariables(Particles.RibbonWidth=&SelectFloatFromArray(Direct Array Index=$Particles.PartitionParticles.Partition, Float Selection Array=DataInterface() {propertyValues {FloatData [15,15]}})) {isSetParameters=true, module=SetVariables_#14}
   /SetVariables(Particles.RibbonLength:NiagaraFloat=&SelectFloatFromArray(Direct Array Index=$Particles.PartitionParticles.Partition, Float Selection Array=DataInterface() {propertyValues {FloatData [155,155]}})) {isSetParameters=true, module=SetVariables_#15}
  ]
  particleUpdateScript [
   &ParticleState(Kill Particles When Lifetime Has Elapsed=false)
   &SocketLocation(
    Orientation Sampling=::NewEnumerator1
    Position Sampling=::NewEnumerator1
    SamplePosOffset=true
    Sampled Position Offset=&SelectVectorFromArray(
     Direct Array Index=$Particles.PartitionParticles.Partition
     Vector Selection Array=DataInterface() {propertyValues {FloatData [<0 0 -5>, <0 0 -5>]}}
    )
    Sampled Position Offset Coordinate Space:ENiagaraSkelMeshTransforms=::NewEnumerator2
    Socket Index=$Particles.PartitionParticles.Partition
    Socket Reader=$System.Socket Reader Feet
    Socket Sampling Mode=::NewEnumerator3
    VelocitySpeedLimit=false
   )
   /SetVariables(Particles.RibbonLength:NiagaraFloat=&Multiply_Float(
     A=$Particles.Initial.RibbonLength
     B=&FloatFromCurve(
      FloatCurve=DataInterface() {propertyValues {Curve {keys [
          {arriveTangent=1, interpMode=RCIM_Cubic, time=0, value=0}
          {interpMode=RCIM_Cubic, tangentMode=RCTM_User, time=0, value=0}
          {interpMode=RCIM_Cubic, tangentMode=RCTM_User, time=1, value=1}
         ]}}}
      Scale Curve=&Lerp_Float(A=1, B=1.2, Alpha=$System.ContrailDiveAlpha)
     )
    )) {isSetParameters=true, module=SetVariables_#16}
   /SetVariables(Particles.InterpolatedTrailVector:Vector3f=$Emitter.InterpolatedVector) {isSetParameters=true, module=SetVariables_#17}
   /SetVariables(Particles.InterpolatedPosition:NiagaraPosition=&AddVectorToPosition(
     Position=$Output.SocketLocation.SampledPosition
     Vector=&Multiply_VectorByFloat(
      Vector=$Particles.InterpolatedTrailVector
      Float=&Multiply_Float(A=$Particles.RibbonLength, B=$Particles.PartitionParticles.Normalized Index In Partition)
     )
    )) {isSetParameters=true, module=SetVariables_#18}
   &ScaleRibbonWidth(Ribbon Width Scale=&Multiply_Float(
     A=&FloatFromCurve(
      FloatCurve=DataInterface() {propertyValues {Curve {keys [
          {interpMode=RCIM_Cubic, tangentMode=RCTM_User, time=0, value=0}
          {arriveTangent=1.4972, interpMode=RCIM_Cubic, leaveTangent=1.4972, tangentMode=RCTM_User, time=1, value=1.3}
         ]}}}
      Scale Curve=1.3
     )
     B=&MergeHelper_FloatSwitch(
      0=&FloatFromCurve(
       FloatCurve=DataInterface() {propertyValues {Curve {keys [
           {interpMode=RCIM_Cubic, tangentMode=RCTM_User, time=0, value=0.2}
           {arriveTangentWeight=0.5252, interpMode=RCIM_Cubic, tangentMode=RCTM_User, time=1, value=1}
          ]}}}
       CurveIndex=$Particles.RibbonLinkOrder
       Scale Curve=1.5
      )
      1=&FloatFromCurve(
       FloatCurve=DataInterface() {propertyValues {Curve {keys [
           {interpMode=RCIM_Cubic, tangentMode=RCTM_User, time=0, value=0.2}
           {arriveTangentWeight=0.5252, interpMode=RCIM_Cubic, tangentMode=RCTM_User, time=1, value=1}
          ]}}}
       CurveIndex=$Particles.RibbonLinkOrder
       Scale Curve=1.5
      )
      2=&FloatFromCurve(
       FloatCurve=DataInterface() {propertyValues {Curve {keys [
           {interpMode=RCIM_Cubic, tangentMode=RCTM_User, time=0, value=0.5}
           {arriveTangentWeight=0.5252, interpMode=RCIM_Cubic, tangentMode=RCTM_User, time=1, value=1}
          ]}}}
       CurveIndex=$Particles.RibbonLinkOrder
      )
      3=&FloatFromCurve(
       FloatCurve=DataInterface() {propertyValues {Curve {keys [
           {interpMode=RCIM_Cubic, tangentMode=RCTM_User, time=0, value=0.5}
           {arriveTangentWeight=0.5252, interpMode=RCIM_Cubic, tangentMode=RCTM_User, time=1, value=1}
          ]}}}
       CurveIndex=$Particles.RibbonLinkOrder
      )
      Index=$Particles.PartitionParticles.Partition
      Num=::NewEnumerator0
     )
    ))
   &Color(Color=&ColorFromCurve(
     ColorCurve=DataInterface() {propertyValues {
       AlphaCurve {keys [<at 0 0>,<at 0.2858 1.5 RCIM_Cubic>,<at 0.9164 0 RCIM_Cubic>]}
       BlueCurve {keys [<at 0.0111 0.8177>,<at 0.3546 0.5>,<at 0.663 0.1 RCIM_Cubic>]}
       GreenCurve {keys [<at 0.0111 0.6534>,<at 0.3546 0.2489>,<at 0.663 0.0351 RCIM_Cubic>]}
       RedCurve {keys [<at 0.0111 0.1704>,<at 0.3546 0.0198>,<at 0.663 0.005 RCIM_Cubic>]}
      }}
     CurveIndex=$Particles.RibbonLinkOrder
    ))
   /SetVariables(Particles.Offset X:NiagaraFloat=&Add_Float(A=$Particles.Offset X, B=&Lerp_Float(A=0, B=-0.03, Alpha=$System.ContrailDiveAlpha))) {isSetParameters=true, module=SetVariables_#19}
   &DynamicMaterialParameters(
    Index 0 Param 1=&SelectFloatFromArray(Direct Array Index=$Particles.PartitionParticles.Partition, Float Selection Array=DataInterface() {propertyValues {FloatData [0.7,0.8]}})
    Index 0 Param 2=$Particles.Offset X
    Index 0 Param 3=&Multiply_Float(
     A=&FloatFromCurve(
      FloatCurve=DataInterface() {propertyValues {Curve {keys [
          {arriveTangent=0.0484, interpMode=RCIM_Cubic, leaveTangent=0.0484, tangentMode=RCTM_User, time=0.3, value=1}
          {arriveTangent=0.0924, interpMode=RCIM_Cubic, leaveTangent=0.0924, tangentMode=RCTM_User, time=0.8, value=1}
          <at 1 0 RCIM_Cubic>
         ]}}}
      CurveIndex=$Particles.PartitionParticles.Normalized Index In Partition
      Scale Curve=0.6835
     )
     B=&SelectFloatFromArray(Direct Array Index=$Particles.PartitionParticles.Partition, Float Selection Array=DataInterface() {propertyValues {FloatData [0.5,0.5]}})
    )
    Index 0 Param 4=1
    Module.Index0Float0Write=true
    Module.Index0Float1Write=true
    Module.Index0Float2Write=true
    Module.Index0Float3Write=true
   )
   /SetVariables(Particles.Position=$Particles.InterpolatedPosition) {isSetParameters=true, module=SetVariables_#20}
  ]
  rendererClasses [&NiagaraRibbonRendererProperties]
  renderers [{rendererClass=&NiagaraRibbonRendererProperties}]
 }
 Water_EdgeFoam {
  emitterSpawnScript [
   /SetVariables(Emitter.InterpolatedVector:Vector3f=<0 0 1>, Emitter.TargetVector:Vector3f=<0 0 1>) {isSetParameters=true, module=SetVariables_#11}
   /SetVariables(Emitter.RibbonLength=222) {isSetParameters=true, module=SetVariables_#12}
   /SetVariables(Emitter.ParticlesPerTrail=5) {isSetParameters=true, module=SetVariables_#13}
  ]
  emitterUpdateScript [
   &EmitterState(
    Inactive Response=::NewEnumerator1
    Life Cycle Mode=::NewEnumerator1
    Loop Behavior=::NewEnumerator1
    Loop Duration Mode=::NewEnumerator1
    UseLoopDelay=false
   )
   &SpawnBurst_Instantaneous(
    Spawn Count=&Multiply_Int(A=4, B=$Emitter.ParticlesPerTrail)
    Use Spawn Probability=false
    UseLoopCountLimit=false
   )
   &self:ScratchModule(Target Vector:Vector3f=$System.LookAtVector) {module=ScratchModule}
  ]
  particleSpawnScript [
   &InitializeParticle(
    Lifetime=1.5
    Position Mode=::NewEnumerator0
    Ribbon Width=15
    Ribbon Width Mode=::NewEnumerator1
    Sprite Size Mode=::NewEnumerator3
    Uniform Sprite Size=5
    Write Material Random=true
   )
   &PartitionParticles(
    Normalized Index In Partition Includes One=true
    Particles Per Partition:NiagaraInt32=$Emitter.ParticlesPerTrail
    Partition Mode:ENiagara_PartitionMode_v2=::NewEnumerator1
    Write Partition Index to Ribbon Link Order:ENiagara_AttributeSamplingApplyOutput=::NewEnumerator0
    Write Partition to Ribbon ID=::NewEnumerator0
   )
   /SetVariables(Particles.RibbonWidth=&SelectFloatFromArray(Direct Array Index=$Particles.PartitionParticles.Partition, Float Selection Array=DataInterface() {propertyValues {FloatData [13,13]}})) {isSetParameters=true, module=SetVariables_#14}
   /SetVariables(Particles.RibbonLength:NiagaraFloat=&SelectFloatFromArray(Direct Array Index=$Particles.PartitionParticles.Partition, Float Selection Array=DataInterface() {propertyValues {FloatData [110,110]}})) {isSetParameters=true, module=SetVariables_#15}
  ]
  particleUpdateScript [
   &ParticleState(Kill Particles When Lifetime Has Elapsed=false)
   &SocketLocation(
    Orientation Sampling=::NewEnumerator1
    Position Sampling=::NewEnumerator1
    SamplePosOffset=true
    Sampled Position Offset=&SelectVectorFromArray(
     Direct Array Index=$Particles.PartitionParticles.Partition
     Vector Selection Array=DataInterface() {propertyValues {FloatData [<0 0 -5>, <0 0 -5>]}}
    )
    Sampled Position Offset Coordinate Space:ENiagaraSkelMeshTransforms=::NewEnumerator2
    Socket Index=$Particles.PartitionParticles.Partition
    Socket Reader=$System.Socket Reader Feet
    Socket Sampling Mode=::NewEnumerator3
    VelocitySpeedLimit=false
   )
   /SetVariables(Particles.RibbonLength:NiagaraFloat=&Multiply_Float(
     A=$Particles.Initial.RibbonLength
     B=&FloatFromCurve(
      FloatCurve=DataInterface() {propertyValues {Curve {keys [{interpMode=RCIM_Cubic, tangentMode=RCTM_User, time=0, value=0}, {interpMode=RCIM_Cubic, tangentMode=RCTM_User, time=1, value=1}]}}}
      Scale Curve=&Lerp_Float(A=1, B=1.2, Alpha=$System.ContrailDiveAlpha)
     )
    )) {isSetParameters=true, module=SetVariables_#16}
   /SetVariables(Particles.InterpolatedTrailVector:Vector3f=$Emitter.InterpolatedVector) {isSetParameters=true, module=SetVariables_#17}
   /SetVariables(Particles.InterpolatedPosition:NiagaraPosition=&AddVectorToPosition(
     Position=$Output.SocketLocation.SampledPosition
     Vector=&Multiply_VectorByFloat(
      Vector=$Particles.InterpolatedTrailVector
      Float=&Multiply_Float(A=$Particles.RibbonLength, B=$Particles.PartitionParticles.Normalized Index In Partition)
     )
    )) {isSetParameters=true, module=SetVariables_#18}
   &ScaleRibbonWidth(Ribbon Width Scale=&Multiply_Float(
     A=&FloatFromCurve(
      FloatCurve=DataInterface() {propertyValues {Curve {keys [
          {interpMode=RCIM_Cubic, tangentMode=RCTM_User, time=0, value=0}
          {arriveTangent=1.4972, interpMode=RCIM_Cubic, leaveTangent=1.4972, tangentMode=RCTM_User, time=1, value=1.3}
         ]}}}
      Scale Curve=1.3
     )
     B=&MergeHelper_FloatSwitch(
      0=&FloatFromCurve(
       FloatCurve=DataInterface() {propertyValues {Curve {keys [
           {interpMode=RCIM_Cubic, tangentMode=RCTM_User, time=0, value=0.2}
           {arriveTangentWeight=0.5252, interpMode=RCIM_Cubic, tangentMode=RCTM_User, time=1, value=1}
          ]}}}
       CurveIndex=$Particles.RibbonLinkOrder
       Scale Curve=1.5
      )
      1=&FloatFromCurve(
       FloatCurve=DataInterface() {propertyValues {Curve {keys [
           {interpMode=RCIM_Cubic, tangentMode=RCTM_User, time=0, value=0.2}
           {arriveTangentWeight=0.5252, interpMode=RCIM_Cubic, tangentMode=RCTM_User, time=1, value=1}
          ]}}}
       CurveIndex=$Particles.RibbonLinkOrder
       Scale Curve=1.5
      )
      2=&FloatFromCurve(
       FloatCurve=DataInterface() {propertyValues {Curve {keys [
           {interpMode=RCIM_Cubic, tangentMode=RCTM_User, time=0, value=0.5}
           {arriveTangentWeight=0.5252, interpMode=RCIM_Cubic, tangentMode=RCTM_User, time=1, value=1}
          ]}}}
       CurveIndex=$Particles.RibbonLinkOrder
      )
      3=&FloatFromCurve(
       FloatCurve=DataInterface() {propertyValues {Curve {keys [
           {interpMode=RCIM_Cubic, tangentMode=RCTM_User, time=0, value=0.5}
           {arriveTangentWeight=0.5252, interpMode=RCIM_Cubic, tangentMode=RCTM_User, time=1, value=1}
          ]}}}
       CurveIndex=$Particles.RibbonLinkOrder
      )
      Index=$Particles.PartitionParticles.Partition
      Num=::NewEnumerator0
     )
    ))
   &Color(Color=&ColorFromCurve(
     ColorCurve=DataInterface() {propertyValues {
       AlphaCurve {keys [<at 0 0>, {arriveTangent=1.2937, interpMode=RCIM_Cubic, leaveTangent=1.2937, time=0.1601, value=1.5}, <at 0.773 1>, <at 1 0 RCIM_Cubic>]}
       BlueCurve {keys [<at 0.4901 0.5521>,<at 0.9949 0.1 RCIM_Cubic>]}
       GreenCurve {keys [<at 0.4901 0.3847>,<at 0.9949 0.0351 RCIM_Cubic>]}
       RedCurve {keys [<at 0.4901 0.138>,<at 0.9949 0.005 RCIM_Cubic>]}
      }}
     CurveIndex=$Particles.RibbonLinkOrder
    ))
   /SetVariables(Particles.Offset X:NiagaraFloat=&Add_Float(A=$Particles.Offset X, B=&Lerp_Float(A=0, B=-0.03, Alpha=$System.ContrailDiveAlpha))) {isSetParameters=true, module=SetVariables_#19}
   &DynamicMaterialParameters(
    Index 0 Param 1=&SelectFloatFromArray(Direct Array Index=$Particles.PartitionParticles.Partition, Float Selection Array=DataInterface() {propertyValues {FloatData [0.5,0.8]}})
    Index 0 Param 2=$Particles.Offset X
    Index 0 Param 3=5
    Index 0 Param 4=1
    Index 1 Param 3=&FloatFromCurve(FloatCurve=DataInterface() {propertyValues {Curve {keys [<at 0 0>,<at 1.0023 0.5>]}}})
    Module.Index0Float0Write=true
    Module.Index0Float1Write=true
    Module.Index0Float2Write=true
    Module.Index0Float3Write=true
    Module.Index1Float0Write=true
    Module.Index1Float1Write=true
    Module.Index1Float2Write=true
    Module.Index1Float3Write=true
    Write Parameter Index 1=::NewEnumerator1
   )
   /SetVariables(Particles.Position=$Particles.InterpolatedPosition) {isSetParameters=true, module=SetVariables_#20}
  ]
  rendererClasses [&NiagaraRibbonRendererProperties]
  renderers [{rendererClass=&NiagaraRibbonRendererProperties}]
 }
 WindMesh {
  emitterUpdateScript [
   &EmitterState(Life Cycle Mode=::NewEnumerator1, Loop Behavior=::NewEnumerator1, Loop Duration Mode=::NewEnumerator1, UseLoopDelay=false)
   &SpawnBurst_Instantaneous(Spawn Count=2, Use Spawn Probability=false, UseLoopCountLimit=false)
  ]
  particleSpawnScript [
   &PartitionParticles()
   &InitializeParticle(
    Color=<rgba 1 1 1 0.35>
    Lifetime=5
    Mesh Scale=<0.3 0.15 0.15>
    Mesh Scale Mode=::NewEnumerator4
    UsePositionOffset=false
    Write Material Random=true
   )
   &InitialMeshOrientation(Mesh Orientation Mode=::NewEnumerator2, Rotation=<0 0.25 0>, Use Rotation Vector=true)
   &DynamicMaterialParameters(
    Index 0 Param 1=1
    Index 0 Param 2=2
    Module.Index0Float0Write=true
    Module.Index0Float1Write=true
    Module.Index0Float2Write=true
    Module.Index0Float3Write=true
   )
  ]
  particleUpdateScript [
   &ParticleState(Kill Particles When Lifetime Has Elapsed=false)
   &SkeletalMeshLocation(
    Bone / Socket Index=&SelectIntFromArray(Direct Array Index=$Particles.PartitionParticles.Partition, Int Selection Array=DataInterface() {propertyValues {IntData [0,1]}})
    Bone Sampling Mode=::NewEnumerator4
    MaskSpawnGroup=false
    Orientation Sampling=::NewEnumerator1
    Sampled Position Offset=&SelectVectorFromArray(
     Direct Array Index=&ReturnExecIndex()
     Vector Selection Array=DataInterface() {propertyValues {FloatData [<0 0 -5>, <0 0 -5>]}}
    )
    Sampled Position Offset Coordinate Space:ENiagaraSkelMeshTransforms=::NewEnumerator0
    Skeletal Mesh=DataInterface() {propertyValues {
      DefaultMesh=null
      ExcludeBoneName=root
      OptionalComponentClass=null
      PreviewMesh=&M_Med_Soldier_04_Preview
      SkinningMode=SkinOnTheFly
      SoftSourceActor=null
      SourceMode=Default
      UvSetIndex=0
      WholeMeshLOD=-1
      bExcludeBone=false
      bReadDeformedGeometry=true
      bRequireCurrentFrameData=true
      ComponentTags []
      FilteredBones [hand_l,hand_r]
      FilteredSockets []
      MeshUserParameter {parameter {name=None, typeDefHandle {registeredTypeIndex=1}, varData []}}
      SamplingRegions []
     }}
   )
   &UpdateMeshOrientation(
    Facing Interpolation=false
    Orientation Method=::NewEnumerator1
    Override Quat=false
    TransformFacing=false
    TransformUp=false
    Up Direction:Vector3f=&FN_FrontEndVector3Modifier(Gameplay=<0 0 1>, Front End=&Lerp_Vector(A=<0 -0.2 1>, B=<0 -0.7 1>, Alpha=$User.ParaGlideLeanAlpha))
   )
   &ScaleColor(
    Scale Alpha=&FN_FrontEndFloatModifier(
     Gameplay=&FNTimeOfDayScalar(Night Scale=2, Morning Scale=1.3, Day Scale=1.3, Evening Scale=1.3)
     Front End=1
    )
    Scale RGB=&MakeVectorFromLinearColorRGB(Color=$System.ToD Color)
    ScaleA=true
    ScaleRGB=true
   )
  ]
  rendererClasses [&NiagaraMeshRendererProperties]
  renderers [{rendererClass=&NiagaraMeshRendererProperties, rendererIndex=0}]
 }
 WristRope {
  emitterUpdateScript [
   &EmitterState()
   &SpawnBurst_Instantaneous(Use Spawn Probability=false, UseLoopCountLimit=false)
  ]
  particleSpawnScript [
   &PartitionParticles(Number Of Partitions=1)
   &InitializeParticle(
    Color=<rgba 0.3646 0.1489 0.0399 0.35>
    Lifetime=&MergeHelper_FloatSwitch(0=5, 1=0.3, 2=?, 3=?, Index=&SpawnGroup(), Num=::NewEnumerator0)
    Mesh Scale Mode=::NewEnumerator3
    Mesh Uniform Scale=&MergeHelper_FloatSwitch(
     0=2.3
     1=&MakeFloatFromVector(Vector=&RandomRangeVector(Minimum=<1.1 1.3 1.3>, Maximum=<1.3 1.4 1.4>), Channel=::NewEnumerator0)
     2=?
     3=?
     Index=&SpawnGroup()
     Num=::NewEnumerator0
    )
    UsePositionOffset=false
    Write Material Random=true
   )
   &InitialMeshOrientation(
    Mesh Axis To Orient=<0 1 0>
    Mesh Orientation Mode=::NewEnumerator2
    Orientation Coordinate Space=::Local
    Rotation=&SelectVectorFromArray(
     Direct Array Index=$Particles.PartitionParticles.Partition
     Vector Selection Array=DataInterface() {propertyValues {FloatData [<0 0.7542 0>, <0 -0.6486 0>]}}
    )
    Use Rotation Vector=true
   )
  ]
  particleUpdateScript [
   &ParticleState(Kill Particles When Lifetime Has Elapsed=false)
   &SkeletalMeshLocation(
    Bone / Socket Index=$Particles.PartitionParticles.Partition
    MaskSpawnGroup=false
    Mesh Sampling Type=::NewEnumerator5
    Sampled Orientation Coordinate Space=::Local
    Sampled Position Offset=&FN_FrontEndVector3Modifier(Gameplay=<1.25 0 0.25>, Front End=&Lerp_Vector(A=<1.25 0 0.25>, B=<0.25 0 2.5>, Alpha=$User.ParaGlideLeanAlpha))
    Sampled Position Offset Coordinate Space:ENiagaraSkelMeshTransforms=::NewEnumerator2
    Skeletal Mesh=DataInterface() {propertyValues {
      DefaultMesh=null
      ExcludeBoneName=root
      OptionalComponentClass=null
      PreviewMesh=&M_Med_Soldier_04_Preview
      SkinningMode=SkinOnTheFly
      SoftSourceActor=null
      SourceMode=Default
      UvSetIndex=0
      WholeMeshLOD=-1
      bExcludeBone=false
      bReadDeformedGeometry=true
      bRequireCurrentFrameData=true
      ComponentTags []
      FilteredBones [hand_l,foot_l,foot_r]
      FilteredSockets [FX_Contrail_Hand_L]
      MeshUserParameter {parameter {name=None, typeDefHandle {registeredTypeIndex=1}, varData []}}
      SamplingRegions []
     }}
    Socket Sampling Mode=::NewEnumerator5
   )
   &DynamicMaterialParameters(Write Parameter Index 0=::NewEnumerator0)
   &ScaleMeshSize(Scale Factor=&MergeHelper_VectorSwitch(
     0=<1 1 1>
     1=&VectorFromCurve(
      VectorCurve=DataInterface() {propertyValues {
        XCurve {keys [{arriveTangent=2.0113, interpMode=RCIM_Cubic, leaveTangent=2.0113, tangentMode=RCTM_User, time=0, value=0}, <at 1 0.7 RCIM_Cubic>]}
        YCurve {keys [{arriveTangent=2.0113, interpMode=RCIM_Cubic, leaveTangent=2.0113, tangentMode=RCTM_User, time=0, value=0}, <at 1 0.7 RCIM_Cubic>]}
        ZCurve {keys [{arriveTangent=2.0113, interpMode=RCIM_Cubic, leaveTangent=2.0113, tangentMode=RCTM_User, time=0, value=0}, <at 1 0.7 RCIM_Cubic>]}
       }}
      Scale Curve=&VectorFromFloat(Value=0.16)
     )
     2=?
     3=?
     Index=&SpawnGroup()
     Num=::NewEnumerator0
    ))
   /SetVariables(Particles.VisibilityTag=&SpawnGroup()) {isSetParameters=true}
  ]
  rendererClasses [&NiagaraMeshRendererProperties]
  renderers [{rendererClass=&NiagaraMeshRendererProperties, rendererIndex=0}]
 }
}
summary {
 systemName=NS_Contrail_SquibblySquid
 emitters [
  {emitterName=Splash_Center, rendererClasses [&NiagaraSpriteRendererProperties]}
  {emitterName=Water_Center, rendererClasses [&NiagaraRibbonRendererProperties]}
  {emitterName=SplashMeshes, rendererClasses [&NiagaraMeshRendererProperties]}
  {emitterName=Water_EdgeFoam, rendererClasses [&NiagaraRibbonRendererProperties]}
  {emitterName=WindMesh, rendererClasses [&NiagaraMeshRendererProperties]}
  {emitterName=RopeRibbon, rendererClasses [&NiagaraRibbonRendererProperties]}
  {emitterName=Tassel, rendererClasses [&NiagaraMeshRendererProperties]}
  {emitterName=WristRope, rendererClasses [&NiagaraMeshRendererProperties]}
  {emitterName=FoamParticles, rendererClasses [&NiagaraSpriteRendererProperties]}
  {emitterName=WaterTrail_Low, rendererClasses [&NiagaraRibbonRendererProperties]}
 ]
 userVariables [
  User.bIsFrontEnd {type=NiagaraBool}
  User.bIsFrontEndPreview {type=NiagaraBool}
  User.ControlCurve {type=NiagaraFloat, defaultValue=0, description=""}
  User.ParaGlideLeanAlpha {type=NiagaraFloat, defaultValue=0, description=""}
  User.PlayerSpeed {type=NiagaraFloat, defaultValue=0, description=""}
  User.RibbonIDOffset {type=NiagaraInt32}
  User.SkeletalMesh {
   type=NiagaraDataInterfaceSkeletalMesh
   defaultValue {
    struct=NiagaraExt_VariableValue_DataInterface
    value {dataInterface=&self:NiagaraDataInterfaceSkeletalMesh_0}
   }
  }
  User.V2 {type=NiagaraFloat, defaultValue=0, description=""}
 ]
}
